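# Conversation database for the Power Chord NPC in NPCLibrary, defining the greeting and menu branches shown when a player talks to the character.
# The entry point feeds one speech node that selects contextual lines by priority curves for first-ever conversations and specific worn outfits, alongside general lines gated on match conversation count.
# The speech node then branches through link nodes to the shared NPC quests, basic services and Power Chord dialogue entry tags.
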ConversationDatabase /NPCLibrary/NPCs/PowerChord/PowerChord/NPCLibrary_Conversation_PowerChord
entry_points [{tag=Conversation.Entry.PowerChord, destinations [#1]}]
nodes [
 ConversationEntryPointNode_7 {
  id=#1
  class=ConversationEntryPointNode
  outputs [#2]
  properties {entryTag {tagName=Conversation.Entry.PowerChord}}
 }
 Speech_C_4 {
  id=#2
  class=Speech_C
  outputs [#3, #4, #5]
  properties {
   nodeName=""
   generalConfig {
    defaultMessage="Who would want to be a marauder, anyway?"
    contextualMessages [
     {
      message="You know what this island needs? MORE MUSIC!"
      contextRequirements [{requirement {class=FortControllerRequirement_IsFirstEverConversationWithNPC}}]
      priority {curve {rowName=Tandem.Speech.Priority.FirstEverInteraction}}
     }
     {
      message="Who would want to be a marauder, anyway?"
      contextRequirements [{requirement {class=FortControllerRequirement_CheckNPCMatchConversationCount, properties {nPCUniqueIdTag {tagName=AISpawnerData.Type.Tandem.PowerChord}}}}]
      priority {curve {rowName=None}}
     }
     {
      message="SHHH, you hear that? The wind's got some serious tempo!"
      contextRequirements [{requirement {class=FortControllerRequirement_CheckNPCMatchConversationCount, properties {nPCUniqueIdTag {tagName=AISpawnerData.Type.Tandem.PowerChord}}}}]
      priority {curve {rowName=None}}
     }
     {
      message="Megalo Don? Nah, the Slides are where it's at."
      contextRequirements [{requirement {class=FortControllerRequirement_CheckNPCMatchConversationCount, properties {nPCUniqueIdTag {tagName=AISpawnerData.Type.Tandem.PowerChord}}}}]
      priority {curve {rowName=None}}
     }
     {
      message="This storm is SERIOUSLY punk rock!"
      contextRequirements [{requirement {class=FortControllerRequirement_CheckNPCMatchConversationCount, properties {nPCUniqueIdTag {tagName=AISpawnerData.Type.Tandem.PowerChord}}}}]
      priority {curve {rowName=None}}
     }
     {
      message="You gotta LISTEN. The Slides is SUCH a better name!"
      contextRequirements [{requirement {
         class=FortControllerRequirement_HasCID
         properties {allowedCIds [/BRCosmetics/Athena/Items/Cosmetics/Characters/CID_196_Athena_Commando_M_Biker]}
        }}]
      priority {curve {rowName=Tandem.Speech.Priority.OutfitReaction}}
     }
     {
      message=Ruckus...sometimes I wonder about you.
      contextRequirements [{requirement {
         class=FortControllerRequirement_HasCID
         properties {allowedCIds [/BRCosmetics/Athena/Items/Cosmetics/Characters/CID_262_Athena_Commando_M_MadCommander]}
        }}]
      priority {curve {rowName=Tandem.Speech.Priority.OutfitReaction}}
     }
     {
      message="FINALLY, someone who gets MUSIC!"
      contextRequirements [{requirement {
         class=FortControllerRequirement_HasCID
         properties {allowedCIds [/BRCosmetics/Athena/Items/Cosmetics/Characters/CID_268_Athena_Commando_M_RockerPunk]}
        }}]
      priority {curve {rowName=Tandem.Speech.Priority.OutfitReaction}}
     }
     {
      message="Woah, wait. Who're you?"
      contextRequirements [{requirement {
         class=FortControllerRequirement_HasCID
         properties {allowedCIds [/BRCosmetics/Athena/Items/Cosmetics/Characters/CID_A_097_Athena_Commando_F_WastelandWarrior]}
        }}]
      priority {curve {rowName=Tandem.Speech.Priority.OutfitReaction}}
     }
    ]
   }
  }
 }
 ConversationLinkNode_2 {
  id=#5
  class=ConversationLinkNode
  properties {remoteEntryTag {tagName=Conversation.Entry.PowerChord.Dialogue}}
 }
 ConversationLinkNode_0 {id=#4, class=ConversationLinkNode, properties {remoteEntryTag {tagName=Conversation.Entry.BasicServices}}}
 ConversationLinkNode_1 {id=#3, class=ConversationLinkNode, properties {remoteEntryTag {tagName=Conversation.Entry.NPCLibrary.Quests}}}
]
